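# NPC_Glitch_Cue is an Athena sound cue, named for an NPC glitch sound, that contains no sound nodes and has no root.
# Only playback settings are configured: NPC_Glitch_Att attenuation, the Fort_Player_Foley_Athena sound class and reduced volume.
# A concurrency override allows one instance at a time and stops the oldest.
# With an empty graph, the cue as shown has nothing to play.
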SoundCue /Game/Athena/Sounds/NPCs/NPC_Glitch_Cue
root=null
nodes []
properties {
 attenuationSettings=&/NPC_Glitch_Att
 bOverrideConcurrency=true
 maxDistance=1200
 soundClassObject=&Fort_Player_Foley_Athena
 volumeMultiplier=0.5
 concurrencyOverrides {maxCount=1, resolutionRule=StopOldest}
}
